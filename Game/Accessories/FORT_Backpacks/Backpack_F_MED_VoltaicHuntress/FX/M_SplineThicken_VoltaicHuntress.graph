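# Translucent unlit trail material for Niagara mesh particles in the Voltaic Huntress backpack effect.
# SplineThicken widens the strip, while panning Perlin noise and a clamped, tail-weighted pull toward a Niagara-supplied position add world position offset.
# Opacity combines a distorted smoke texture, end fades, noise dissolve, particle alpha and quality-switched depth fade.
# Color blends two tints by UV and smoke, plus optional panning detail.
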Material /Game/Accessories/FORT_Backpacks/Backpack_F_MED_VoltaicHuntress/FX/M_SplineThicken_VoltaicHuntress
graph {
 properties {
  bUsedWithNiagaraMeshParticles=true
  blendMode=BLEND_Translucent
  pixelDepthOffsetMode=PDOM_Legacy
  shadingModel=MSM_Unlit
  twoSided=true
  usedShadingModels=Unlit
 }
 nodes [
  Constant() {r=10}
  n1 = ScalarParameter() {defaultValue=2, group=SplineThicken, parameterName=SplineThickenWidth}
  n2 = Multiply(
   A=ScalarParameter() {defaultValue=5, group=SplineThicken, parameterName=ThickenCorePower}
   B=MaterialFunctionCall!(UVs, Offset, RotationCenter, RotationAngle=Constant() {r=0.25}) {materialFunction=/FortUILibrary/MaterialFunctions/Gradients/MF_UI_ReflectedGradient}.Result
  )
  n3 = Multiply(
   A=NamedRerouteDeclaration(None=Saturate(None=@n2)) {name=EdgeGradient, nodeColor=<rgba 1 0 0.9141 1>}
   B=ScalarParameter() {defaultValue=1, group=SplineThicken, parameterName=ThickenCoreSize}
  )
  n4 = ScalarParameter() {group=SplineThicken, parameterName=ThickenCore?}
  n5 = ScalarParameter() {defaultValue=15, group=SplineThicken, parameterName=SplineThickenWidthTip}
  n6 = MaterialFunctionCall(
   WidthBase=LinearInterpolate(A=@n1, B=Multiply(A=@n1, B=@n3), Alpha=@n4)
   WidthTip=LinearInterpolate(A=@n5, B=Multiply(A=@n3, B=@n5), Alpha=@n4)
  ) {materialFunction=&SplineThicken}
  n7 = Multiply(
   A=ScalarParameter() {defaultValue=1, group=SplineThicken, parameterName=Spline Width Mult}
   B=DynamicParameter() {defaultValue=<rgba 0 1 1 1>, paramNames [Random Offset, Spline Width Mult, Dissolve, Move UV]}.Spline Width Mult
  )
  NamedRerouteDeclaration(None=StaticSwitchParameter(True=Multiply(A=@n6.WorldPositionOffset, B=@n7), False=Constant()) {defaultValue=true, group=SplineThicken, parameterName=UseSplineThicken}) {name=SplineThicken, nodeColor=<rgba 1 0.6797 0 1>}
  n8 = NamedRerouteUsage()
  n9 = NamedRerouteUsage()
  NamedRerouteDeclaration(None=Add(A=Add(A=@n8, B=@n9) {constB=0}, B=NamedRerouteUsage()) {constB=0}) {name=WPO, nodeColor=<rgba 0.0781 0 1 1>}
  n10 = ComponentMask(None=TextureCoordinate()) {g=true}
  n11 = Power(Base=@n10)
  n12 = ComponentMask(None=TextureCoordinate()) {g=true}
  n13 = Constant() {r=2}
  n14 = Multiply(A=Saturate(None=Multiply(A=@n11) {constB=6}), B=Saturate(None=Multiply(A=@n12, B=@n13) {constB=3}))
  n15 = Power(Base=@n10, Exp=ScalarParameter() {defaultValue=10, group=Opacity, parameterName=Bottom Fade Power})
  n16 = Saturate(None=Multiply(A=OneMinus(None=@n15), B=ScalarParameter() {defaultValue=3, group=Opacity, parameterName=Bottom Fade Mult}) {constB=6})
  n17 = Constant()
  n18 = ParticleRandom()
  n19 = VectorParameter() {group=Opacity, parameterName=SmokeTrail UV, sortPriority=1, defaultValue=<rgba 1 1 0 0.4>}
  n20 = ScalarParameter() {defaultValue=1, group=Noise, parameterName=Global Scale U}
  n21 = ScalarParameter() {defaultValue=1, group=Noise, parameterName=Global Scale V}
  n22 = Multiply(A=TextureCoordinate(), B=Multiply(A=AppendVector(A=@n19.R, B=@n19.R), B=AppendVector(A=@n20, B=@n21)))
  n23 = AppendVector(A=@n19.B, B=@n19.B)
  n24 = Add(A=Add(A=AppendVector(A=@n17, B=@n18), B=Panner(Coordinate=@n22, Speed=@n23)), B=NamedRerouteUsage())
  n25 = NamedRerouteUsage()
  n26 = TextureSampleParameter2D(UVs=MaterialFunctionCall(UVs=Add(A=@n24, B=@n25), "Rotation Angle (0-1)"=Constant() {r=0.25}) {materialFunction=&CustomRotator}.Rotated Values) {group=Opacity, parameterName=SmokeTrailTexture, texture=/Game/Effects/Fort_Effects/Textures/Smoke/T_Smoke_Trail_Soft_Chromatic}
  n27 = LinearInterpolate(A=@n26.G, B=Multiply(A=@n26.R), Alpha=ScalarParameter() {group=Opacity, parameterName=Use soft trail?})
  n28 = VectorParameter() {group=Color, parameterName=Refine, defaultValue=<rgba 6 10 1 0.9>}
  n29 = Saturate(None=MaterialFunctionCall(
    Texture=@n27
    RGB=MaterialFunctionCall(X=@n28.R, Y=@n28.R, Z=@n28.R) {materialFunction=&MakeFloat3}.Result
    A=@n28.A
   ) {materialFunction=&MF_FX_Refine}.Result)
  n30 = ScalarParameter() {defaultValue=0.7, group=Opacity, parameterName=Opacity Multiplier}
  n31 = Power(
   Base=Multiply(A=Multiply(A=@n29, B=@n30), B=Multiply(A=@n14, B=@n16))
   Exp=ScalarParameter() {defaultValue=1, group=Opacity, parameterName=Opacity Power}
  )
  n32 = ScalarParameter() {defaultValue=2, group=SplineThicken, parameterName=ThickenCorePowerOpacity}
  n33 = MaterialFunctionCall(RotationAngle=Constant() {r=0.25}) {materialFunction=&MF_UI_ReflectedGradient}
  n34 = Subtract(A=Saturate(None=Multiply(A=@n32, B=@n33.Result)), B=ScalarParameter())
  n35 = Multiply(
   A=LinearInterpolate(A=@n31, B=Multiply(A=@n31, B=@n34), Alpha=ScalarParameter() {group=SplineThicken, parameterName=ThickenCore?})
   B=LinearInterpolate(B=NamedRerouteUsage(), Alpha=ScalarParameter() {defaultValue=1, parameterName=useDissolve?}) {constA=1}
  )
  n36 = ParticleColor()
  n37 = DepthFade(FadeDistance=ScalarParameter() {group=Opacity, parameterName=FadeDistance})
  n38 = Constant() {r=1}
  NamedRerouteDeclaration(None=Multiply(A=Multiply(A=@n35, B=@n36.A), B=QualitySwitch(Default=@n37, Low=@n38, High=@n37, Medium=@n38, Epic=@n37))) {name=Opacity, nodeColor=<rgba 0.9219 1 0 1>}
  n39 = ParticleRandom()
  n40 = VectorParameter() {group=Distortion, parameterName=Distortion UV, sortPriority=1, defaultValue=<rgba 1 1 0 0.1>}
  n41 = ScalarParameter() {defaultValue=1, group=Noise, parameterName=Global Scale U}
  n42 = ScalarParameter() {defaultValue=1, group=Noise, parameterName=Global Scale V}
  n43 = Multiply(A=TextureCoordinate(), B=Multiply(A=AppendVector(A=@n40.R, B=@n40.R), B=AppendVector(A=@n41, B=@n42)))
  n44 = AppendVector(A=@n40.B, B=@n40.B)
  n45 = TextureSampleParameter2D(UVs=Add(A=AppendVector(A=@n39, B=@n39), B=Panner(Coordinate=@n43, Speed=@n44))) {
   group=Distortion
   parameterName="Distortion  Tex"
   samplerType=SAMPLERTYPE_Normal
   sortPriority=0
   texture=&T_Weap_Distort_N_02
  }
  NamedRerouteDeclaration(None=Multiply(
    A=ComponentMask(None=@n45.RGB) {g=true, r=true}
    B=ScalarParameter() {defaultValue=0.2, group=Distortion, parameterName=Distortion Intensity, sortPriority=2}
   )) {name=Distortion, nodeColor=<rgba 1 0 0.375 1>}
  n46 = NamedRerouteUsage()
  n47 = ScalarParameter() {parameterName=Use detail?}
  n48 = MaterialFunctionCall(Float2=TextureCoordinate()) {materialFunction=&BreakOutFloat2Components}
  n49 = Constant() {r=6}
  n50 = Constant() {r=6}
  n51 = LinearInterpolate(
   A=VectorParameter() {group=Color, parameterName=ColorA, defaultValue=<rgba 0.0221 0.1186 0.849 1>}.RGB
   B=VectorParameter() {group=Color, parameterName=ColorB, defaultValue=<rgba 0 1.2 1.18 1>}.RGB
   Alpha=LinearInterpolate(A=Power(Base=@n48.G, Exp=@n49), B=Power(Base=@n27, Exp=@n50), Alpha=Constant() {r=0.13}) {constAlpha=0.2}
  )
  n52 = ScalarParameter() {defaultValue=1, group=Color, parameterName=Color Multiplier}
  NamedRerouteDeclaration(None=Add(A=LinearInterpolate(B=@n46, Alpha=@n47), B=Multiply(A=@n51, B=@n52))) {name=Color, nodeColor=<rgba 1 0.375 0 1>}
  n53 = Constant()
  n54 = ScalarParameter() {parameterName=RandomWPO_Offset}
  n55 = VectorParameter() {
   group=WPO
   parameterName=WPO Depth UV
   sortPriority=6
   channelNames {a=WPO Depth V Pan, b=WPO Depth U Pan, g=WPO Depth V, r=WPO Depth U}
   defaultValue=<rgba 0 0.6 0 0.2>
  }
  n56 = Multiply(A=TextureCoordinate(), B=AppendVector(A=@n55.WPO Depth U, B=@n55.WPO Depth U))
  n57 = AppendVector(A=@n55.WPO Depth U Pan, B=@n55.WPO Depth U Pan)
  n58 = Add(A=Add(A=AppendVector(A=@n54, B=@n54), B=Panner(Coordinate=@n56, Speed=@n57)), B=NamedRerouteUsage())
  n59 = Multiply(
   A=LinearInterpolate(
    A=ScalarParameter() {group=WPO, parameterName=WPO Z Clamp}
    Alpha=TextureSample(UVs=@n58) {samplerType=SAMPLERTYPE_LinearColor, texture=/Game/Global/Textures/Default/Tileable_Patterns/T_PerlinNoise_Uncompressed}.R
   ) {constB=-1}
   B=ScalarParameter() {defaultValue=10, group=WPO, parameterName=WPO Z Noise Amount}
  ) {constB=6}
  n60 = ComponentMask(None=TextureCoordinate()) {g=true}
  n61 = OneMinus(None=@n60)
  n62 = Constant() {r=1}
  n63 = Constant() {r=1}
  n64 = Multiply(A=Multiply(A=Power(Base=@n61, Exp=@n62) {constExponent=0.5}, B=Power(Base=@n60, Exp=@n63) {constExponent=0.5}), B=Constant() {r=2}) {constB=4}
  n65 = DynamicParameter() {parameterIndex=1, defaultValue=<rgba 0 10 1 1>, paramNames [Random Offset, Noise Intensty, Dissolve, Dissolve Edge Intensity]}
  n66 = LinearInterpolate(
   A=ScalarParameter() {defaultValue=3, group=WPO, parameterName=WPO Noise Intensity}
   B=@n65.Noise Intensty
   Alpha=ScalarParameter() {group=WPO, parameterName=UseDynNoiseIntensity}
  )
  n67 = Multiply(
   A=Multiply(
    A=MaterialFunctionCall(R=@n53, G=@n59, B=@n53) {materialFunction=&AppendMany}.RGB
    B=Saturate(None=@n64)
   )
   B=@n66
  )
  n68 = VectorParameter() {group=WPO, parameterName=WPO Noise UV, sortPriority=6, defaultValue=<rgba 0 0.65 0 0.1>}
  n69 = TextureCoordinate()
  n70 = AppendVector(A=ComponentMask(None=@n69) {r=true}, B=Power(Base=ComponentMask(None=@n69) {g=true}) {constExponent=1})
  n71 = Multiply(A=@n70, B=AppendVector(A=@n68.R, B=@n68.R))
  n72 = AppendVector(A=@n68.B, B=@n68.B)
  n73 = ScalarParameter() {parameterName=RandomWPO_Offset}
  n74 = AppendVector(A=@n73, B=@n73)
  n75 = Add(A=Add(A=Panner(Coordinate=@n71, Speed=@n72), B=@n74), B=NamedRerouteUsage())
  n76 = TextureObject() {samplerType=SAMPLERTYPE_LinearColor, texture=&T_PerlinNoise_Uncompressed}
  n77 = Multiply(
   A=LinearInterpolate(Alpha=TextureSample(UVs=@n75, Tex=@n76) {samplerType=SAMPLERTYPE_LinearColor, texture=&T_PerlinNoise_Uncompressed}.R) {constA=-1}
   B=ScalarParameter() {defaultValue=5, group=WPO, parameterName=WPO X Noise Amount}
  )
  n78 = Constant()
  n79 = VectorParameter() {group=WPO, parameterName=WPO Noise UV 2, sortPriority=6, defaultValue=<rgba 0 0.5 0 0.2>}
  n80 = Multiply(A=@n70, B=AppendVector(A=@n79.R, B=@n79.R))
  n81 = AppendVector(A=@n79.B, B=@n79.B)
  n82 = Add(A=Add(A=Panner(Coordinate=@n80, Speed=@n81), B=@n74), B=NamedRerouteUsage())
  n83 = Multiply(
   A=LinearInterpolate(Alpha=TextureSample(UVs=@n82, Tex=@n76) {samplerType=SAMPLERTYPE_LinearColor, texture=&T_PerlinNoise_Uncompressed}.R) {constA=-1}
   B=ScalarParameter() {defaultValue=1, group=WPO, parameterName=WPO Y Noise Amount}
  ) {constB=0.5}
  n84 = ComponentMask(None=TextureCoordinate()) {g=true}
  n85 = OneMinus(None=@n84)
  n86 = Constant() {r=1.8}
  n87 = Power(Base=@n84, Exp=Constant() {r=0.9}) {constExponent=0.5}
  n88 = ScalarParameter() {group=SplineThicken, parameterName=WPO Top Mask}
  n89 = Multiply(
   A=Multiply(
    A=MaterialFunctionCall(R=@n77, G=@n78, B=@n83) {materialFunction=&AppendMany}.RGB
    B=@n66
   )
   B=Multiply(A=Power(Base=@n85, Exp=@n86) {constExponent=0.5}, B=LinearInterpolate(B=@n87, Alpha=@n88) {constA=1})
  )
  NamedRerouteDeclaration(None=Transform(None=Add(A=@n67, B=@n89) {constB=0}) {transformSourceType=TRANSFORMSOURCE_Local}) {name=WPO Distortion, nodeColor=<rgba 1 0 0.9844 1>}
  n90 = DynamicParameter() {desc=Parameters set by Niagara, parameterIndex=3, defaultValue=<rgba 0 1 1 1>, paramNames [X,Y,Z,Intensity]}
  n91 = MaterialFunctionCall(Float3=Subtract(
    A=MaterialFunctionCall(X=@n90.X, Y=@n90.X, Z=@n90.X) {materialFunction=&MakeFloat3}.Result
    B=ParticlePositionWS()
   )) {materialFunction=&BreakOutFloat3Components}
  n92 = ScalarParameter() {defaultValue=30, group=Interpolate, parameterName=Interpolate Clamp XY}
  n93 = ScalarParameter() {defaultValue=1, group=Interpolate, parameterName=Interpolate Clamp Z}
  n94 = AppendVector(
   A=Clamp(None=AppendVector(A=@n91.R, B=@n91.R), Min=Multiply(A=@n92) {constB=-1}, Max=@n92) {maxDefault=1000, minDefault=-1000}
   B=Clamp(None=@n91.B, Min=Multiply(A=@n93) {constB=-1}, Max=@n93) {minDefault=-1}
  )
  n95 = OneMinus(None=TextureCoordinate())
  n96 = Multiply(
   A=@n90.Intensity
   B=Power(Base=ComponentMask(None=@n95) {g=true}, Exp=ScalarParameter() {defaultValue=2.5, group=Interpolate, parameterName=Interpolate Power})
  )
  NamedRerouteDeclaration(None=StaticSwitchParameter(True=LinearInterpolate(B=@n94, Alpha=@n96), False=Constant()) {defaultValue=true, group=Interpolate, parameterName=Use Interpolate WPO, sortPriority=0}) {name=Interpolate WPO, nodeColor=<rgba 0.3828 1 0 1>}
  n97 = ParticleRandom()
  n98 = AppendVector(A=@n97, B=@n97)
  n99 = VectorParameter() {group=Emissive, parameterName=Emissive Detail UV, sortPriority=1, defaultValue=<rgba 0.55 0.5 0 0.2>}
  n100 = AppendVector(A=@n99.R, B=@n99.R)
  n101 = AppendVector(
   A=ScalarParameter() {defaultValue=1, group=Noise, parameterName=Global Scale U}
   B=ScalarParameter() {defaultValue=1, group=Noise, parameterName=Global Scale V}
  )
  n102 = Panner(Coordinate=Multiply(A=TextureCoordinate(), B=Multiply(A=@n100, B=@n101)), Speed=AppendVector(A=@n99.B, B=@n99.B))
  n103 = TextureSample(UVs=Add(A=Add(A=@n98, B=@n102))) {texture=&T_Tile_RainbowColor}
  n104 = ScalarParameter() {parameterName=DesaturateDetail}
  NamedRerouteDeclaration(None=Multiply(
    A=Desaturation(None=@n103.RGB, Fraction=@n104)
    B=ScalarParameter() {defaultValue=1, group=Emissive, parameterName=Emissive Detail Intensity}
   )) {name=Detail, nodeColor=<rgba 1 0 0.7031 1>}
  n105 = TextureCoordinate()
  n106 = Multiply(A=ComponentMask(None=@n105) {r=true}, B=OneMinus(None=ComponentMask(None=@n105) {r=true}))
  n107 = Multiply(
   A=LinearInterpolate(B=Multiply(A=@n106) {constB=2}, Alpha=ScalarParameter() {defaultValue=-1.5, group=Dissolve, parameterName=Lerp Gradient Tip}) {constA=1, constAlpha=1}
   B=ComponentMask(None=TextureCoordinate()) {g=true}
  )
  n108 = ParticleRandom()
  n109 = AppendVector(A=@n108, B=@n108)
  n110 = VectorParameter() {group=Dissolve, parameterName=Dissolve Tex UV, sortPriority=6, defaultValue=<rgba 0.3 0.5 0 0.2>}
  n111 = Panner(Coordinate=Multiply(A=TextureCoordinate(), B=AppendVector(A=@n110.R, B=@n110.R)), Speed=AppendVector(A=@n110.B, B=@n110.B))
  n112 = Multiply(
   A=@n107
   B=TextureSampleParameter2D(UVs=Add(A=@n109, B=@n111)) {group=Dissolve, parameterName=Dissolve Tex, texture=/Game/Global/Textures/Default/Noise/T_FireNoise_01}.R
  )
  n113 = ScalarParameter() {defaultValue=1, group=Dissolve, parameterName=Lerp Dissolve Tex}
  n114 = ScalarParameter() {defaultValue=1, group=Dissolve, parameterName=Dissolve Edge Hardness, sliderMin=0.5}
  n115 = Multiply(
   A=Power(Base=LinearInterpolate(A=@n107, B=@n112, Alpha=@n113), Exp=ScalarParameter() {defaultValue=1, group=Dissolve, parameterName=Dissolve Power})
   B=@n114
  )
  n116 = MaterialFunctionCall() {materialFunction=&LinearGradient}
  n117 = Constant() {r=1.5}
  n118 = LinearInterpolate(A=@n114, B=Constant() {r=-1}, Alpha=Saturate(None=Multiply(A=@n116.VGradient, B=@n117))) {constB=-1}
  NamedRerouteDeclaration(None=Saturate(None=Subtract(A=@n115, B=@n118))) {name=Dissolve, nodeColor=<rgba 1 0 0.9609 1>}
  NamedRerouteDeclaration(None=MaterialFunctionCall(X=Constant(), Y=DynamicParameter() {defaultValue=<rgba 0 1 1 1>, paramNames [Random Offset, Spline Width Mult, Dissolve, Move UV]}.Move UV) {materialFunction=&MakeFloat2}.Result) {name=UV Additional Movement, nodeColor=<rgba 0 1 0.7422 1>}
  NamedRerouteUsage()
  NamedRerouteUsage()
  NamedRerouteUsage()
  NamedRerouteUsage()
 ]
}
